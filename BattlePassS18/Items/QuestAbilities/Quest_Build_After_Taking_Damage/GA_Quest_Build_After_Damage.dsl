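; Quest gameplay ability that credits a player for building shortly after taking damage.
; On event activation, presumably from damage, it stores the hit time and binds to the player controller's spawned-building delegate, ending if the avatar cast fails.
; Each build within the allowed window sends an Athena.Quests.BuildAfterDamage stat event to the quest manager.
; A build after the window ends the ability.
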(event Ability|EventActivateAbilityFromEvent
  (params (EventData :type Gameplay%20Event%20Data%20Structure%20%28by%20ref%29))
  (|CallParentFunction EventData)
  (bind _asfort_player_pawn_athena (Utilities|Casting|CastToFortPlayerPawnAthena (Ability|GetAvatarActorfromActorInfo))
    (:then
      (CallFunction|HandleTookDamage)
      (Building|AssignOnPlayerSpawnedBuildingActor (Fortnite|GetFortPlayerControllerfromActor _asfort_player_pawn_athena) (delegate Custom|OnPlayerSpawnedBuildingActor_Event_0)))
    (:CastFailed
      (Ability|EndAbility))))

(event Custom|OnPlayerSpawnedBuildingActor_Event_0
  (params (Spawner :type Fort%20Player%20Controller%20Object%20Reference) (SpawnedBuilding :type Building%20Actor%20Object%20Reference))
  (CallFunction|HandleSpawnedBuilding Spawner SpawnedBuilding))

(fn HandleSpawnedBuilding
  (params (Spawner%20PC :type Fort%20Player%20Controller%20Object%20Reference) (Spawned%20Building :type Building%20Actor%20Object%20Reference))
  (if (<= (- (Utilities|Time|GetTimeSeconds) (Variables|Default|GetTimeLastHit)) (Variables|Default|GetTimeAllowedAfterDamage))
    (Quest|SendComplexCustomStatEvent (Variables|GAVars|Data|GetOwningPlayerQuestManager) (Variables|GetSpawned%20Building) "(GameplayTags=((TagName=\"Athena.Quests.BuildAfterDamage\")))" "(GameplayTags=)" 1)
    (Ability|EndAbility)))

(fn HandleTookDamage
  (Variables|Default|SetTimeLastHit (Utilities|Time|GetTimeSeconds)))
